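# Material function that layers optional detail color and normal textures, defaulting to rock details, onto incoming material attributes.
# A static switch picks object-radius-scaled UVs or world-aligned Z projection; the Z-only toggle has identical branches.
# Detail color overlays BaseColor, its red channel modulates Specular and Roughness, and scaled normals are added and normalized.
# Forward and mobile paths change only BaseColor using UV sampling; a second output exposes detail color alone.
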MaterialFunction /Game/Environments/Apollo/VS/Materials/MF_DetailTexture
graph {nodes [
  n1 = GetMaterialAttributes(None=FunctionInput() {inputType=FunctionInput_MaterialAttributes}) {attributeGetTypes [
    #1
    #2
    #3
    #4
   ]}
  n2 = Multiply(
   A=TextureCoordinate()
   B=Divide(A=ObjectRadius(), B=ScalarParameter() {defaultValue=300, group=Detail Texture, parameterName=DetailTile, sortPriority=3})
  )
  n3 = TextureObjectParameter() {
   group=Detail Texture
   parameterName=Detail D
   sortPriority=1
   texture=/Game/Environments/Apollo/Rocks/Natural_Rocks/Textures/T_RockDetails_02D
  }
  n4 = TextureSample(UVs=@n2, Tex=@n3)
  n5 = ScalarParameter() {defaultValue=768, group=Detail Texture, parameterName=DetailTile}
  n6 = MaterialFunctionCall(TextureObject=@n3, TextureSize=@n5) {materialFunction=/Game/Weapons/WeaponSkins/Wraps/MaterialFunctions/WorldAlignedTexture}
  n7 = StaticBoolParameter() {group=Detail Texture, parameterName=Use Z only projection}
  n8 = StaticSwitch(True=@n6.Z Texture, False=@n6.Z Texture, Value=@n7)
  n9 = StaticSwitchParameter(True=@n4.RGB, False=ComponentMask(None=@n8) {b=true, g=true, r=true}) {defaultValue=true, group=Detail Texture, parameterName=Use Object Radius Details, sortPriority=-1}
  n10 = ShadingPathSwitch(Default=@n9, Deferred=@n9, Forward=@n4.RGB, Mobile=@n4.RGB)
  n11 = MaterialFunctionCall(Base=@n1.BaseColor, Blend=@n10) {materialFunction=&Blend_Overlay}
  n12 = StaticSwitchParameter(True=@n4.R, False=ComponentMask(None=@n8) {r=true}) {defaultValue=true, group=Detail Texture, parameterName=Use Object Radius Details, sortPriority=-1}
  n13 = ScalarParameter() {defaultValue=-0.773, group=Detail Texture, parameterName=Detail R Min, sortPriority=5}
  n14 = ScalarParameter() {defaultValue=1.8112, group=Detail Texture, parameterName=Detail R Max, sortPriority=6}
  n15 = TextureObjectParameter() {
   group=Detail Texture
   parameterName=Detail N
   samplerType=SAMPLERTYPE_Normal
   sortPriority=2
   texture=/Game/Environments/Apollo/Rocks/Natural_Rocks/Textures/T_RockDetails_02_N
  }
  n16 = VectorParameter() {group=Detail Texture, parameterName=N Details, sortPriority=4, defaultValue=<rgba 0.4 0.4 0 1>}
  n17 = MaterialFunctionCall(TextureObject=@n15, TextureSize=@n5) {materialFunction=&WorldAlignedNormal}
  n18 = StaticSwitchParameter(
   True=Multiply(
    A=TextureSample(UVs=@n2, Tex=@n15) {samplerType=SAMPLERTYPE_Normal, texture=&T_RockDetails_02_N}.RGB
    B=@n16.RGB
   )
   False=Multiply(A=StaticSwitch(True=@n17.Z Texture, False=@n17.Z Texture, Value=@n7), B=@n16.RGB)
  ) {defaultValue=true, group=Detail Texture, parameterName=Use Object Radius Details, sortPriority=-1}
  n19 = SetMaterialAttributes!(
   MaterialAttributes=@n1.MaterialAttributes
   BaseColor=@n11.Result
   Specular=Multiply(A=@n12, B=@n1.Specular)
   Roughness=Multiply(A=@n1.Roughness, B=LinearInterpolate(A=@n13, B=@n14, Alpha=@n12))
   Normal=Normalize(VectorInput=Add(A=@n1.Normal, B=@n18))
  ) {attributeSetTypes [
    #1
    #2
    #3
    #4
   ]}
  n20 = SetMaterialAttributes!(MaterialAttributes=@n1.MaterialAttributes, BaseColor=@n11.Result) {attributeSetTypes [#1]}
  FunctionOutput(None=StaticSwitchParameter(True=ShadingPathSwitch(Default=@n19, Deferred=@n19, Forward=@n20, Mobile=@n20), False=@n1.MaterialAttributes) {defaultValue=true, group=Detail Texture, parameterName="Detail Textures "})
  FunctionOutput(None=StaticSwitchParameter(True=@n10, False=Constant() {r=0.5}) {defaultValue=true, group=Detail Texture, parameterName="Detail Textures "}) {outputName=Texture Only}
 ]}
