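# Material function for Apollo cliffs that builds world-aligned surface attributes with a tinted waterline.
# WorldAlignedTextures_Complex projects diffuse, specular and normal inputs with height-refined transitions, exporting diffuse alpha as a height map.
# Spec red drives both Specular and Roughness, and the attributes pass through MF_DetailTexture.
# Base color below a world-Z height is multiplied by waterline color and detail texture red, fading out above over a blur distance.
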MaterialFunction /Game/Environments/Apollo/VS/Materials/MF_Apollo_Cliff
graph {nodes [
  n1 = FunctionInput() {bUsePreviewValueAsDefault=true, inputName=Tile, inputType=FunctionInput_Scalar, previewValue=<359 0 0 1>}
  n2 = StaticBool() {value=true}
  n3 = FunctionInput() {bUsePreviewValueAsDefault=true, inputName=Normal, inputType=FunctionInput_Texture2D, previewValue=<0 0 1 1>}
  n4 = MaterialFunctionCall!(
   Diffuse Texture Object=FunctionInput() {bUsePreviewValueAsDefault=true, inputName=Diffuse, inputType=FunctionInput_Texture2D}
   Spec Texture Object=FunctionInput() {bUsePreviewValueAsDefault=true, inputName=Specular, inputType=FunctionInput_Texture2D}
   TextureSize=@n1
   --------------------------
   Use seperate settings for Z projection
   Use seperate texture for Z projection
   Texture Size Z projection
   Z Plane Diffuse Texture Object
   Z Plane Spec Texture Object
   "--------------------------      "
   ProjectionTransitionContrast=Constant() {r=0.7178}
   Use Texture Alpha Heightmap to Refine the Transistion=@n2
   HeightLerpTransitionContrast=Constant() {r=0.8838}
   PickTheDiffuseHeightChannel=@n2
   TheDiffuseHeightChannel=Constant4Vector() {constant=<rgba 0 0 0 1>}.RGBA
   "--------------------------       "
   NormalTextureObject=@n3
   World Space Normals=StaticBool()
   "--------------------------  "
   Export Float 4 Diffuse=@n2
   Export Float 4 Spec
   "--------------------------     "
   WorldPosition
   World Space Normal
  ) {materialFunction=/Engine/Functions/Engine_MaterialFunctions02/Texturing/WorldAlignedTextures_Complex}
  FunctionOutput(None=ComponentMask(None=@n4.XY Diffuse Texture) {a=true}) {outputName=Height map 3D}
  n5 = MaterialFunctionCall(Float3=@n4.XY Spec) {materialFunction=&BreakOutFloat3Components}
  n6 = SetMaterialAttributes!(
   MaterialAttributes
   BaseColor=Multiply(
    A=ComponentMask(None=@n4.XY Diffuse Texture) {b=true, g=true, r=true}
    B=FunctionInput() {bUsePreviewValueAsDefault=true, inputName=BaseColorTone Walls, previewValue=<1 1 1 1>}
   )
   Specular=@n5.R
   Roughness=@n5.R
   Normal=Multiply(A=@n4.Normal XY Texture, B=FunctionInput() {bUsePreviewValueAsDefault=true, inputName=Nomal Value, previewValue=<1 1 1 1>})
  ) {attributeSetTypes [
    #1
    #2
    #3
    #4
   ]}
  n7 = MaterialFunctionCall!(In=@n6) {materialFunction=&/MF_DetailTexture}
  n8 = GetMaterialAttributes(None=@n7.Result) {attributeGetTypes [#1]}
  n9 = Multiply(
   A=@n8.BaseColor
   B=Multiply(
    A=VectorParameter() {group=Cliffs Waterline, parameterName=Cliffs Waterline, defaultValue=<rgba 0.9115 0.7863 0.6314 0>}.RGB
    B=ComponentMask(None=@n7.Texture Only) {r=true}
   )
  )
  n10 = Subtract(A=WorldPosition().XYZ, B=ScalarParameter() {defaultValue=21.6059, group=Cliffs Waterline, parameterName=Cliffs Waterline Height})
  n11 = ScalarParameter() {defaultValue=401.7054, group=Cliffs Waterline, parameterName=Cliffs Waterline Blur}
  n12 = Saturate(None=ComponentMask(None=Divide(A=@n10, B=@n11)) {b=true})
  FunctionOutput(None=SetMaterialAttributes!(MaterialAttributes=@n8.MaterialAttributes, BaseColor=LinearInterpolate(A=@n9, B=@n8.BaseColor, Alpha=@n12)) {attributeSetTypes [#1]}) {outputName=Result 3D}
 ]}
